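# Sound cue for outdoor snow-biome wind rush one-shot ambient spots, using wind attenuation and an outdoor weather sound class at reduced volume.
# A quality-level node sends its first two levels to a random pick of twelve wind rush waves and its last two to a smaller pick of four (takes 01, 02, 06, 07).
# Each branch passes through a modulator that slightly varies pitch and volume.
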SoundCue /Game/Sounds/Artemis/Biomes/Snow/Spots/Spot_Snow_Wind_Rush_Cue
root=QualityLevel_1
nodes [
 {id=Modulator_3, children [Random_6], properties {pitchMax=1.1, pitchMin=0.9, volumeMax=1.1, volumeMin=0.9}}
 {id=Modulator_5, children [Random_7], properties {pitchMax=1.1, pitchMin=0.9, volumeMax=1.1, volumeMin=0.9}}
 {id=QualityLevel_1, children [Modulator_3,Modulator_3,Modulator_5,Modulator_5]}
 {
  id=Random_6
  children [
   WavePlayer_20
   WavePlayer_27
   WavePlayer_28
   WavePlayer_29
   WavePlayer_30
   WavePlayer_31
   WavePlayer_32
   WavePlayer_33
   WavePlayer_34
   WavePlayer_35
   WavePlayer_36
   WavePlayer_37
  ]
  properties {weights [1,1,1,1,1,1,1,1,1,1,1,1]}
 }
 {id=Random_7, children [WavePlayer_38, WavePlayer_39, WavePlayer_40, WavePlayer_41], properties {weights [1,1,1,1]}}
 {id=WavePlayer_20, properties {soundWaveAssetPtr=&/Spot_Snow_Wind_Rush_01}}
 {id=WavePlayer_27, properties {soundWaveAssetPtr=&/Spot_Snow_Wind_Rush_02}}
 {id=WavePlayer_28, properties {soundWaveAssetPtr=&/Spot_Snow_Wind_Rush_03}}
 {id=WavePlayer_29, properties {soundWaveAssetPtr=&/Spot_Snow_Wind_Rush_04}}
 {id=WavePlayer_30, properties {soundWaveAssetPtr=&/Spot_Snow_Wind_Rush_05}}
 {id=WavePlayer_31, properties {soundWaveAssetPtr=&/Spot_Snow_Wind_Rush_06}}
 {id=WavePlayer_32, properties {soundWaveAssetPtr=&/Spot_Snow_Wind_Rush_07}}
 {id=WavePlayer_33, properties {soundWaveAssetPtr=&/Spot_Snow_Wind_Rush_08}}
 {id=WavePlayer_34, properties {soundWaveAssetPtr=&/Spot_Snow_Wind_Rush_09}}
 {id=WavePlayer_35, properties {soundWaveAssetPtr=&/Spot_Snow_Wind_Rush_10}}
 {id=WavePlayer_36, properties {soundWaveAssetPtr=&/Spot_Snow_Wind_Rush_11}}
 {id=WavePlayer_37, properties {soundWaveAssetPtr=&/Spot_Snow_Wind_Rush_12}}
 {id=WavePlayer_38, properties {soundWaveAssetPtr=&Spot_Snow_Wind_Rush_07}}
 {id=WavePlayer_39, properties {soundWaveAssetPtr=&Spot_Snow_Wind_Rush_02}}
 {id=WavePlayer_40, properties {soundWaveAssetPtr=&Spot_Snow_Wind_Rush_01}}
 {id=WavePlayer_41, properties {soundWaveAssetPtr=&Spot_Snow_Wind_Rush_06}}
]
properties {
 attenuationSettings=/Game/Sounds/Attenuation/Fort_TileAmbient_OneShot_Wind_Att
 bEnableBusSends=false
 maxDistance=8000
 soundClassObject=/Game/Sounds/FortSoundClassesAndModes/Fort_Ambient_Outdoors_Weather_Athena
 volumeMultiplier=0.251
 preEffectBusSends [{}]
}
